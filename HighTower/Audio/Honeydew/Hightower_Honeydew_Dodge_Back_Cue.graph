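# Sound cue for Honeydew's backward dodge roll in the HighTower content.
# A single wave player for the BackRoll wave feeds a Modulator at the root, which is left at default settings.
# The cue plays in the melee weapon sound class with a volume boost and borrows a close-range Carmine weapon attenuation for spatial falloff.
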SoundCue /HighTower/Audio/Honeydew/Hightower_Honeydew_Dodge_Back_Cue
root=Modulator_0
nodes [
 {id=Modulator_0, children [WavePlayer_0]}
 {id=WavePlayer_0, properties {soundWaveAssetPtr=/HighTower/Audio/Honeydew/waves/Hightower_Honeydew_BackRoll}}
]
properties {
 attenuationSettings=/Game/Sounds/Attenuation/Fort_Weapon_Carmine_Close_Athena_Att
 maxDistance=4500
 soundClassObject=&Fort_WeaponMelee
 volumeMultiplier=1.3
}
