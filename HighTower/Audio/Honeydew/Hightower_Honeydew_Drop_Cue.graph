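# Sound cue for the Honeydew drop sound in HighTower, built from two Honeydew pickup wave variants.
# A random node picks between the two equally weighted waves and feeds a modulator at the root.
# It plays at reduced volume with weapon-drop attenuation and the loot-collect sound class.
# Its concurrency override prevents new instances from starting once the limit is reached.
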SoundCue /HighTower/Audio/Honeydew/Hightower_Honeydew_Drop_Cue
root=Modulator_0
nodes [
 {id=Modulator_0, children [Random_0]}
 {id=Random_0, children [WavePlayer_0,WavePlayer_1], properties {weights [1,1]}}
 {id=WavePlayer_0, properties {soundWaveAssetPtr=/HighTower/Audio/Honeydew/waves/Hightower_Honeydew_Pickup_01}}
 {id=WavePlayer_1, properties {soundWaveAssetPtr=/HighTower/Audio/Honeydew/waves/Hightower_Honeydew_Pickup_02}}
]
properties {
 attenuationSettings=&Fort_Weapon_Drop_Att
 bOverrideConcurrency=true
 maxDistance=3000
 soundClassObject=&Fort_Loot_Collect
 volumeMultiplier=0.5
 concurrencyOverrides {resolutionRule=PreventNew}
}
