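# Sound cue for the Honeydew equip sound in HighTower, routed to the player foley sound class.
# It randomly picks one of two equally weighted aura equip waves.
# A modulator boosts and varies the volume before the shared default HighTower aura attenuation is applied.
# An overridden concurrency setting caps simultaneous instances and stops the quietest when the limit is exceeded.
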SoundCue /HighTower/Audio/Honeydew/Hightower_Honeydew_Equip_Cue
root=Attenuation_0
nodes [
 {id=Attenuation_0, children [Modulator_0], properties {attenuationSettings=/HighTower/Audio/Aura/Default/Hightower_Aura_Att}}
 {id=Modulator_0, children [Random_0], properties {volumeMax=2, volumeMin=1.5}}
 {id=Random_0, children [WavePlayer_0,WavePlayer_3], properties {weights [1,1]}}
 {id=WavePlayer_0, properties {soundWaveAssetPtr=/HighTower/Audio/Aura/Honeydew/Hightower_Honeydew_Aura_Equip_01}}
 {id=WavePlayer_3, properties {soundWaveAssetPtr=/HighTower/Audio/Aura/Honeydew/Hightower_Honeydew_Aura_Equip_02}}
]
properties {
 bOverrideConcurrency=true
 maxDistance=650
 soundClassObject=&Fort_Player_Foley_Athena
 volumeMultiplier=1
 concurrencyOverrides {maxCount=3, resolutionRule=StopQuietest}
}
