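# Sound cue for Honeydew fist impacts in HighTower, routed to the impacts sound class at reduced volume.
# A short randomized delay feeds a modulator, which plays one of six equally weighted impact waves.
# Distance falloff references a shared melee attenuation asset.
# Overridden concurrency caps simultaneous voices per owner and stops the oldest, taking precedence over the assigned melee weapon concurrency set.
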SoundCue /HighTower/Audio/Honeydew/Hightower_Honeydew_Fist_Impact_Cue
root=Delay_1
nodes [
 {id=Delay_1, children [Modulator_0], properties {delayMax=0.1, delayMin=0.05}}
 {id=Modulator_0, children [Random_0]}
 {
  id=Random_0
  children [WavePlayer_0, WavePlayer_1, WavePlayer_2, WavePlayer_3, WavePlayer_4, WavePlayer_10]
  properties {weights [1,1,1,1,1,1]}
 }
 {id=WavePlayer_0, properties {soundWaveAssetPtr=/HighTower/Audio/Honeydew/waves/Hightower_Honeydew_Fist_Impact_01}}
 {id=WavePlayer_1, properties {soundWaveAssetPtr=/HighTower/Audio/Honeydew/waves/Hightower_Honeydew_Fist_Impact_02}}
 {id=WavePlayer_10, properties {soundWaveAssetPtr=/HighTower/Audio/Honeydew/waves/Hightower_Honeydew_Fist_Impact_06}}
 {id=WavePlayer_2, properties {soundWaveAssetPtr=/HighTower/Audio/Honeydew/waves/Hightower_Honeydew_Fist_Impact_03}}
 {id=WavePlayer_3, properties {soundWaveAssetPtr=/HighTower/Audio/Honeydew/waves/Hightower_Honeydew_Fist_Impact_04}}
 {id=WavePlayer_4, properties {soundWaveAssetPtr=/HighTower/Audio/Honeydew/waves/Hightower_Honeydew_Fist_Impact_05}}
]
properties {
 attenuationSettings=/Game/Sounds/Attenuation/Fort_Wasabi_Melee_Att
 bOverrideConcurrency=true
 maxDistance=3800
 soundClassObject=&Fort_Impacts
 volumeMultiplier=0.4
 attenuationOverrides {
  bAttenuateWithLPF=true
  distanceAlgorithm=Logarithmic
  falloffDistance=800
  lPFRadiusMax=800
  lPFRadiusMin=400
  attenuationShapeExtents {x=300}
 }
 concurrencyOverrides {bLimitToOwner=true, maxCount=3, resolutionRule=StopOldest, voiceStealReleaseTime=0.1, volumeScale=0.2}
 concurrencySet [&Player_Melee_Weapon_Conc]
}
