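# Sound cue for the fourth Honeydew fist swing melee sound in HighTower.
# A Random node picks one of three equally weighted swing wave variations.
# A Modulator at the root adds slight random pitch and volume variation to each play.
# The cue uses the melee weapon sound class, the Wasabi melee attenuation asset and player melee weapon concurrency.
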SoundCue /HighTower/Audio/Honeydew/Hightower_Honeydew_Fist_Swing_04_Cue
root=Modulator_7
nodes [
 {id=Modulator_7, children [Random_0], properties {pitchMax=1.02, pitchMin=0.98, volumeMax=1.1, volumeMin=1}}
 {id=Random_0, children [WavePlayer_2,WavePlayer_3,WavePlayer_4], properties {weights [1,1,1]}}
 {id=WavePlayer_2, properties {soundWaveAssetPtr=/HighTower/Audio/Honeydew/waves/Hightower_Honeydew_Fist_Swing_04_01}}
 {id=WavePlayer_3, properties {soundWaveAssetPtr=/HighTower/Audio/Honeydew/waves/Hightower_Honeydew_Fist_Swing_04_02}}
 {id=WavePlayer_4, properties {soundWaveAssetPtr=/HighTower/Audio/Honeydew/waves/Hightower_Honeydew_Fist_Swing_04_03}}
]
properties {
 attenuationSettings=/Game/Sounds/Attenuation/Fort_Wasabi_Melee_Att
 maxDistance=3800
 soundClassObject=&Fort_WeaponMelee
 attenuationOverrides {distanceAlgorithm=NaturalSound, falloffDistance=2000, attenuationShapeExtents {x=500}}
 concurrencySet [&Player_Melee_Weapon_Conc]
}
