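# Sound cue for the smash on Honeydew's fourth fist swing, a HighTower melee attack sound.
# A Random node picks between two equally weighted fist slam waves and feeds a Modulator that adds very slight pitch variation.
# It routes to the melee weapon sound class, references the shared Wasabi melee attenuation asset, and uses player melee weapon concurrency to govern overlapping plays.
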SoundCue /HighTower/Audio/Honeydew/Hightower_Honeydew_Fist_Swing_04_Smash_Cue
root=Modulator_0
nodes [
 {id=Modulator_0, children [Random_1], properties {pitchMax=1.01, pitchMin=0.99, volumeMin=1}}
 {id=Random_1, children [WavePlayer_0,WavePlayer_1], properties {weights [1,1]}}
 {id=WavePlayer_0, properties {soundWaveAssetPtr=/HighTower/Audio/Honeydew/waves/Hightower_Honeydew_Fist_Slam_01}}
 {id=WavePlayer_1, properties {soundWaveAssetPtr=/HighTower/Audio/Honeydew/waves/Hightower_Honeydew_Fist_Slam_02}}
]
properties {
 attenuationSettings=/Game/Sounds/Attenuation/Fort_Wasabi_Melee_Att
 maxDistance=3800
 soundClassObject=&Fort_WeaponMelee
 attenuationOverrides {distanceAlgorithm=NaturalSound, falloffDistance=2000, attenuationShapeExtents {x=500}}
 concurrencySet [&Player_Melee_Weapon_Conc]
}
