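# Sound cue for the Honeydew jump activation in HighTower, playing a jump burst one-shot.
# A random node picks between two equally weighted burst waves, then a modulator applies slight pitch and volume variation.
# It uses close-range weapon projectile attenuation and the melee weapon sound class, with a boosted overall volume.
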SoundCue /HighTower/Audio/Honeydew/Hightower_Honeydew_Jump_Activate_Cue
root=Modulator_0
nodes [
 {id=Modulator_0, children [Random_0], properties {pitchMax=1.1, pitchMin=0.9, volumeMax=1.1, volumeMin=1}}
 {id=Random_0, children [WavePlayer_4,WavePlayer_5], properties {weights [1,1]}}
 {id=WavePlayer_4, properties {soundWaveAssetPtr=/HighTower/Audio/Honeydew/waves/Hightower_Honeydew_Jump_Burst_01}}
 {id=WavePlayer_5, properties {soundWaveAssetPtr=/HighTower/Audio/Honeydew/waves/Hightower_Honeydew_Jump_Burst_02}}
]
properties {
 attenuationSettings=&Fort_Weapon_Projectile_Close_Athena_Att
 maxDistance=4950
 soundClassObject=&Fort_WeaponMelee
 volumeMultiplier=1.25
}
