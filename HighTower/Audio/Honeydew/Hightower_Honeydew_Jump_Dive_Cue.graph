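# Sound cue for Honeydew's jump dive in HighTower, layering a random one-of-three start sound with a looping dive sound.
# It mixes a first-person branch using close projectile attenuation with a quieter third-person branch using far attenuation, each under its own weapon-fire sound class.
# In both branches an enveloper fades the loop in, more slowly in third person.
# Modulators vary pitch per branch, and the third-person start is also pitched down.
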SoundCue /HighTower/Audio/Honeydew/Hightower_Honeydew_Jump_Dive_Cue
root=Mixer_0
nodes [
 {
  id=Attenuation_1
  children [Modulator_0]
  properties {attenuationSettings=&Fort_Weapon_Projectile_Close_Athena_Att, attenuationOverrides {bSpatialize=false}}
 }
 {
  id=Attenuation_2
  children [Modulator_1]
  properties {attenuationSettings=/Game/Sounds/Attenuation/Fort_Weapon_Carmine_Far_Athena_Att, attenuationOverrides {bSpatialize=false}}
 }
 {
  id=Enveloper_0
  children [WavePlayer_3]
  properties {pitchCurve {editorCurveData {keys [{time=0}]}}, volumeCurve {editorCurveData {keys [<at 0 0.01>,<at 1 1>]}}}
 }
 {
  id=Enveloper_1
  children [WavePlayer_7]
  properties {pitchCurve {editorCurveData {keys [{time=0}]}}, volumeCurve {editorCurveData {keys [<at 0 0.01>,<at 1.5 1>]}}}
 }
 {id=Mixer_0, children [SoundClass_0,SoundClass_1], properties {inputVolume [1,0.15]}}
 {id=Mixer_1, children [Random_1,Enveloper_0], properties {inputVolume [1,0.2]}}
 {id=Mixer_2, children [Modulator_2,Enveloper_1], properties {inputVolume [1,1]}}
 {id=Modulator_0, children [Mixer_1], properties {pitchMax=1.1}}
 {id=Modulator_1, children [Mixer_2], properties {pitchMax=1.1, pitchMin=0.9, volumeMax=0.6, volumeMin=0.5}}
 {id=Modulator_2, children [Random_0], properties {pitchMax=0.85, pitchMin=0.8}}
 {id=Random_0, children [WavePlayer_4,WavePlayer_5,WavePlayer_6], properties {weights [1,1,1]}}
 {id=Random_1, children [WavePlayer_0,WavePlayer_1,WavePlayer_2], properties {weights [1,1,1]}}
 {
  id=SoundClass_0
  children [Attenuation_1]
  properties {soundClassOverride=&Fort_WeaponRanged_Fire_1P_PSM}
 }
 {
  id=SoundClass_1
  children [Attenuation_2]
  properties {soundClassOverride=&Fort_WeaponRanged_Fire_3P_PSM}
 }
 {id=WavePlayer_0, properties {soundWaveAssetPtr=/HighTower/Audio/Honeydew/waves/Hightower_Honeydew_Jump_Dive_Start_01}}
 {id=WavePlayer_1, properties {soundWaveAssetPtr=/HighTower/Audio/Honeydew/waves/Hightower_Honeydew_Jump_Dive_Start_02}}
 {id=WavePlayer_2, properties {soundWaveAssetPtr=/HighTower/Audio/Honeydew/waves/Hightower_Honeydew_Jump_Dive_Start_03}}
 {id=WavePlayer_3, properties {bLooping=true, soundWaveAssetPtr=/HighTower/Audio/Honeydew/waves/Hightower_Honeydew_Dive_Loop}}
 {id=WavePlayer_4, properties {soundWaveAssetPtr=&Hightower_Honeydew_Jump_Dive_Start_01}}
 {id=WavePlayer_5, properties {soundWaveAssetPtr=&Hightower_Honeydew_Jump_Dive_Start_02}}
 {id=WavePlayer_6, properties {soundWaveAssetPtr=&Hightower_Honeydew_Jump_Dive_Start_03}}
 {id=WavePlayer_7, properties {bLooping=true, soundWaveAssetPtr=&Hightower_Honeydew_Dive_Loop}}
]
properties {maxDistance=25000, volumeMultiplier=1}
