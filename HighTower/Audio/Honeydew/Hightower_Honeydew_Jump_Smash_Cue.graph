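# Sound cue for the Honeydew jump smash in HighTower, mixing close and distant explosion-attenuation layers, both with spatialization overridden off.
# Each layer randomly picks from the same three smash waves and routes through a third-person ranged weapon fire sound class.
# The close layer varies pitch around normal, while the distant layer reuses the waves pitched down and quieter to act as a far tail.
# Concurrency allows one instance per owner, stopping the oldest.
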SoundCue /HighTower/Audio/Honeydew/Hightower_Honeydew_Jump_Smash_Cue
root=Mixer_0
nodes [
 {
  id=Attenuation_1
  children [Random_0]
  properties {attenuationSettings=/Game/Sounds/Attenuation/Fort_Weapon_Milo_Close_Explosion_Athena_Att, attenuationOverrides {bSpatialize=false}}
 }
 {
  id=Attenuation_2
  children [Modulator_1]
  properties {attenuationSettings=/Game/Sounds/Attenuation/Fort_Weapon_Milo_Distant_Explosion_Athena_Att, attenuationOverrides {bSpatialize=false}}
 }
 {id=Mixer_0, children [Modulator_0,SoundClass_1], properties {inputVolume [0.75,0.75]}}
 {id=Modulator_0, children [SoundClass_0], properties {pitchMax=1.1, pitchMin=0.9}}
 {id=Modulator_1, children [Random_1], properties {pitchMax=0.85, pitchMin=0.8, volumeMax=0.55, volumeMin=0.5}}
 {id=Random_0, children [WavePlayer_2,WavePlayer_5,WavePlayer_6], properties {weights [1,1,1]}}
 {id=Random_1, children [WavePlayer_0,WavePlayer_1,WavePlayer_3], properties {weights [1,1,1]}}
 {
  id=SoundClass_0
  children [Attenuation_1]
  properties {soundClassOverride=&Fort_WeaponRanged_Fire_3P_PSM}
 }
 {
  id=SoundClass_1
  children [Attenuation_2]
  properties {soundClassOverride=&Fort_WeaponRanged_Fire_3P_PSM}
 }
 {id=WavePlayer_0, properties {soundWaveAssetPtr=/HighTower/Audio/Honeydew/waves/Hightower_Honeydew_Jump_Smash_01}}
 {id=WavePlayer_1, properties {soundWaveAssetPtr=/HighTower/Audio/Honeydew/waves/Hightower_Honeydew_Jump_Smash_02}}
 {id=WavePlayer_2, properties {soundWaveAssetPtr=&Hightower_Honeydew_Jump_Smash_01}}
 {id=WavePlayer_3, properties {soundWaveAssetPtr=/HighTower/Audio/Honeydew/waves/Hightower_Honeydew_Jump_Smash_03}}
 {id=WavePlayer_5, properties {soundWaveAssetPtr=&Hightower_Honeydew_Jump_Smash_02}}
 {id=WavePlayer_6, properties {soundWaveAssetPtr=&Hightower_Honeydew_Jump_Smash_03}}
]
properties {
 bOverrideConcurrency=true
 maxDistance=15000
 volumeMultiplier=0.9
 concurrencyOverrides {bLimitToOwner=true, maxCount=1, resolutionRule=StopOldest}
}
