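# Sound cue named for the Honeydew landing sound in HighTower, with no root or sound nodes in its graph, so it has nothing to play as shown.
# It carries only playback settings, borrowing weapon-drop attenuation and the loot-collect sound class at reduced volume.
# A concurrency override uses the PreventNew rule, so new instances are refused once the limit is reached instead of cutting off playing ones.
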SoundCue /HighTower/Audio/Honeydew/Hightower_Honeydew_Land_Cue
root=null
nodes []
properties {
 attenuationSettings=&Fort_Weapon_Drop_Att
 bOverrideConcurrency=true
 maxDistance=3000
 soundClassObject=&Fort_Loot_Collect
 volumeMultiplier=0.5
 concurrencyOverrides {resolutionRule=PreventNew}
}
